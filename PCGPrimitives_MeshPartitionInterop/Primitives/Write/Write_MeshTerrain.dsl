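; PCG primitive graph that writes its input to a Mesh Partition terrain, with the write logic held in a core-process subgraph.
; It tags input as "Input:0", then chains inputs management (processing and fallback), core process and output preparation.
; The core process receives the target partition, bound size, modifier layer and sub-priority, and channel names.
; Graph parameters expose each stage's subgraph, so stages can be swapped without editing the template chain.
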(graph Write_MeshTerrain
  (params (InputsProcessingGraph :type (SoftObject /Script/PCG.PCGGraphInterface) :default /PCGPrimitives/Primitives/_Template_/Primitives_InputsProcessing.Primitives_InputsProcessing) (InputsFallbackGraph :type (SoftObject /Script/PCG.PCGGraphInterface) :default /PCGPrimitives/Primitives/_Template_/Primitives_InputsFallback.Primitives_InputsFallback) (CoreProcessGraph :type (SoftObject /Script/PCG.PCGGraphInterface) :default /PCGPrimitives_MeshPartitionInterop/Primitives/Write/Subgraphs/Write_MeshTerrain_CoreProcess.Write_MeshTerrain_CoreProcess) (PrepareOutputsGraph :type (SoftObject /Script/PCG.PCGGraphInterface) :default /PCGPrimitives/Primitives/_Template_/Primitives_PrepareOutputs.Primitives_PrepareOutputs) (MeshPartitionBoundSize :type Vector :default (1e+05 1e+05 1e+05)) (AffectedMeshPartition :type (SoftObject /Script/MeshPartition.MeshPartition)) (ModifierLayerPriority :type Name) (ModifierSubPriority :type Double) (Channels :type (Array Name)))
  (inputs (PrimaryInput :type Any))
  (bind AddTags_1 (Add%20Tags :TagsToAdd "Input:0"))
  ;[0 0 0 0.5 32] Inputs
  (bind Subgraph_0 (Subgraph :Graph /PCGPrimitives/Primitives/_Template_/Subgraphs/Primitives_InputsManagement.Primitives_InputsManagement :InputsFallbackGraph InputsFallbackGraph :InputsProcessingGraph InputsProcessingGraph))
  ;[0 0 0 0.5 32] Core Process
  (bind Subgraph_3 (Subgraph :Graph /PCGPrimitives_MeshPartitionInterop/Primitives/Write/Subgraphs/Write_MeshTerrain_CoreProcess.Write_MeshTerrain_CoreProcess :AffectedMeshPartition AffectedMeshPartition :Channels Channels :MeshPartitionBoundSize MeshPartitionBoundSize :ModifierPriorityLayer ModifierLayerPriority :ModifierSubPriority ModifierSubPriority :SubgraphOverride-param CoreProcessGraph))
  ;[0 0 0 0.5 32] Prepare Output
  (bind Subgraph_4 (Subgraph :Graph /PCGPrimitives/Primitives/_Template_/Primitives_PrepareOutputs.Primitives_PrepareOutputs :SubgraphOverride-param PrepareOutputsGraph))
  (connect DefaultInputNode PrimaryInput AddTags_1 In)
  (connect AddTags_1 Out Subgraph_0 In)
  (connect Subgraph_0 Out Subgraph_3 In)
  (connect Subgraph_3 Out Subgraph_4 In)
  (output Out Subgraph_4))
